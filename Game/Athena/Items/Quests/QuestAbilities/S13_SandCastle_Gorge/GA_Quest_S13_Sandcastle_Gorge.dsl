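; Gameplay ability for the Season 13 Sandcastle Gorge quest, tracking when its owner leaves the waterfall gorge top state.
; On event activation it calls the parent and prints an on-screen debug message about being inside the gorge.
; It then waits for the waterfall gorge top gameplay tag to be removed, applies the quest's gameplay effect to the owner and ends itself.
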(event Ability|EventActivateAbilityFromEvent
  (params (EventData :type Gameplay%20Event%20Data%20Structure%20%28by%20ref%29))
  (|CallParentFunction EventData)
  (Development|PrintString "AQUA: Inside the gorge!!" true true "(R=0.000000,G=0.660000,B=1.000000,A=1.000000)" 5.0)
  (Ability|Tasks|WaitGameplayTagRemove (Variables|Default|GetT_WaterfallGorge_Top)
    (:Removed
      (Ability|ApplyGameplayEffectToOwner "/Game/Athena/Items/Quests/QuestAbilities/S13_SandCastle_Gorge/GE_Quest_S13_Sandcastle_Gorge.GE_Quest_S13_Sandcastle_Gorge_C")
      (Ability|EndAbility))))
