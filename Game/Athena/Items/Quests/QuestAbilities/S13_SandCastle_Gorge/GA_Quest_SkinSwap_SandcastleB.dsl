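; Quest gameplay ability for the Sandcastle B skin swap, which swaps the player's disguise item when its configured quest objective completes.
; On activation it listens for quest objective changes on the owning controller and checks them against the quest variable.
; On a match it records the character cosmetic's rarity, waits briefly, removes the first swap item, triggers a rarity-levelled effect, grants the Sandcastle B disguise, then unbinds and ends.
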(event Ability|EventActivateAbility
  ;[1 1 1] Check for quest updates
  (|CallParentFunction)
  (EventDispatchers|BindEventtoOnQuestObjectiveStateChanged (Fortnite|GetFortPlayerControllerfromActor (Variables|GAVars|Data|GetOwningPlayerPawn)) (delegate Custom|QuestCompleted)))

(event Custom|QuestCompleted
  (params (Controller :type Fort%20Player%20Controller%20Object%20Reference) (QuestDef :type Fort%20Quest%20Item%20Definition%20Object%20Reference) (BackendName :type Name) (CompletionCount :type Integer) (ObjectiveCompleted :type Boolean) (QuestCompleted :type Boolean))
  ;[1 1 1] Check if the correct quest was completed and change skin.
  (bind _objectivevalid (CallFunction|CheckQuestObjective QuestDef ObjectiveCompleted))
  (if _objectivevalid
    (CallFunction|SetCosmeticRarityLevel)
    (Utilities|FlowControl|Delay 0.45)
    (Inventory|RemoveItemFromPlayer :PlayerController (Fortnite|GetFortPlayerControllerfromActor (Variables|GAVars|Data|GetOwningPlayerPawn)) :ItemDefinition (Variables|Default|GetFirstSwapAGID) :AmountToRemove -1 :bForceRemoval true :bSkipVerbRemoval false)
    (CallFunction|PlayFx)
    (Inventory|GiveItemtoInventoryOwner (Fortnite|GetFortPlayerControllerfromActor (Variables|GAVars|Data|GetOwningPlayerPawn)) "/Game/Athena/Items/Quests/QuestAbilities/S13_SandCastle_Gorge/AGID_Disguise_SandcastleB.AGID_Disguise_SandcastleB")
    (EventDispatchers|UnbindEventfromOnQuestObjectiveStateChanged (Fortnite|GetFortPlayerControllerfromActor (Variables|GAVars|Data|GetOwningPlayerPawn)) (delegate Custom|QuestCompleted))
    (Ability|EndAbility)))

(event Custom|play%20fx
  (GameplayEffects|ApplyGameplayEffectToSelf :self (Ability|GetAbilitySystemComponent (Variables|GAVars|Data|GetOwningPlayerPawn)) :GameplayEffectClass "/Game/Athena/Items/Quests/QuestAbilities/Quest_Alterego/GE_Quest_SkinSwap.GE_Quest_SkinSwap_C" :Level (Variables|GAVars|Data|GetCosmetic%20Rarity%20Level)))

(event Ability|EventFailedToActivatePassiveAbility
  (params (ActorInfo :type Gameplay%20Ability%20Actor%20Info%20Structure))
  (|CallParentFunction))

(fn CheckQuestObjective
  (params (QuestToTest :type Fort%20Quest%20Item%20Definition%20Object%20Reference) (IsObjectiveCompleted :type Boolean))
  (if (Utilities|Equal(Object) QuestToTest (Variables|GAVars|Setup|GetQuest))
    (if IsObjectiveCompleted
      (return true)
      (else
        _)))
  (return false))

(fn SetCosmeticRarityLevel
  (bind _owningplayerpawn (Variables|GAVars|Data|GetOwningPlayerPawn))
  (bind _returnvalue (CosmeticLoadout|GetItemDefinitionfromLoadoutwithActor _owningplayerpawn "(TagName=\"LoadoutSlot.Character\")"))
  (Utilities|IsValid _returnvalue
    (:Is%20Valid
      (Variables|GAVars|Data|SetCosmetic%20Rarity%20Level (Math|Conversions|ToFloat(Byte) (Item|GetRarity _returnvalue))))))

(fn SetNewCosmetics
  (Class|FortPlayerController|ChangeActiveVariantforCosmeticItem (Fortnite|GetFortPlayerControllerfromActor (Variables|GAVars|Data|GetOwningPlayerPawn)) (Variables|GAVars|Setup|GetItem%20Template) (Variables|GAVars|Setup|GetVariant%20Channel) (Variables|GAVars|Setup|GetDesired%20Active%20Variant) 0))
